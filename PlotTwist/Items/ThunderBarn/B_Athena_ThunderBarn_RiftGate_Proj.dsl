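; ThunderBarn Rift Gate Blueprint (named as a projectile) that classifies a player pawn's Force training relative to its own Jedi alignment.
; GetPawnsForceTrainingComparison uses gameplay tags to return one of four enum states: invalid pawn or no Force power, power granted, matching alignment, or differing alignment.
; Alignment compares the pawn's Jedi power tags against bIsThisActorJedi.
; Event graph holds only disabled stubs; the construction script just calls its parent.
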(event EventBeginPlay
  (|CallParentFunction) ; This node is disabled and will not be called.\nDrag off pins to build functionality.
)

(event Collision|EventActorBeginOverlap
  (params (OtherActor :type Actor%20Object%20Reference))
  (|CallParentFunction OtherActor) ; This node is disabled and will not be called.\nDrag off pins to build functionality.
)

(event EventTick
  (params (DeltaSeconds :type Float%20%28single-precision%29))
  (|CallParentFunction DeltaSeconds) ; This node is disabled and will not be called.\nDrag off pins to build functionality.
)

(fn GetPawnsForceTrainingComparison
  (params (FortPlayerPawnToTest :type Fort%20Player%20Pawn%20Object%20Reference))
  (Utilities|IsValid FortPlayerPawnToTest
    (:Is%20Valid
      (Utilities|FlowControl|Sequence
        (:then_0
          (if (not (GameplayTags|HasMatchingGameplayTag (Variables|GetFortPlayerPawnToTest) (Variables|Tags|GetTag_HasForcePower)))
            (return "NewEnumerator0")))
        (:then_1
          (if (GameplayTags|HasMatchingGameplayTag (Variables|GetFortPlayerPawnToTest) (Variables|Tags|GetTag_ForcePowerGranted))
            (return "NewEnumerator1")))
        (:then_2
          (return (select (== (GameplayTags|HasAnyMatchingGameplayTags (Variables|GetFortPlayerPawnToTest) (Variables|Tags|GetTags_JediForcePowers)) (Variables|ProximityMessage|GetbIsThisActorJedi)) "NewEnumerator2" "NewEnumerator3")))))
    (:Is%20Not%20Valid
      (return "NewEnumerator0"))))

(fn ConstructionScript
  (|CallParentFunction))
